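# Niagara system for the S16 vehicle boost cosmetic, drawing exhaust flares, sparks and trails at the car's boost ports.
# Emitters place particles on skeletal mesh sockets per port, shifted by DTCarOffset.
# Two STARLINK cone emitters and rotating sparks form the flare, behind which stream a multi-renderer ribbon with curve-driven widths and colors and GPU sprite particles.
# Painted-variant overrides recolor each layer, while bIsFrontendPreview retunes sizes, rates and speeds for menu preview.
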NiagaraSystem /VehicleCosmetics/Boosts/Boost_S16/FX/NS_Boost_S16
dependencies {
 usedDataInterfaces [
  &NiagaraDataInterfaceCurve
  &NiagaraDataInterfaceArrayColor
  &NiagaraDataInterfaceArrayFloat
  &NiagaraDataInterfaceColorCurve
  &NiagaraDataInterfaceSkeletalMesh
 ]
 usedDynamicInputs []
 usedModules [
  &self:SystemSpawnScript.SystemScriptSource.SystemScriptGraph.NiagaraNodeAssignment_0.SetVariables_#1
  &SystemState
 ]
}
emitters {
 NE_VC_Boost_Cone_STARLINK001 {
  emitterSpawnScript [/SetVariables(Emitter.IS_INGAME:NiagaraBool=&Boolean_NotOperation(Bool=$User.bIsFrontendPreview)) {isSetParameters=true}]
  emitterUpdateScript [
   &EmitterState(
    Inactive Response=::NewEnumerator1
    Life Cycle Mode=::NewEnumerator1
    Loop Behavior=::NewEnumerator1
    Loop Duration Mode=::NewEnumerator1
    UseLoopDelay=false
   )
   &SpawnBurst_Instantaneous(Spawn Count=$System.Initial.PortCount, Spawn Group=1, Use Spawn Probability=false, UseLoopCountLimit=false) {module=SpawnBurst_Instantaneous001}
   &SpawnBurst_Instantaneous(Spawn Count=$System.Initial.PortCount, Use Spawn Probability=false, UseLoopCountLimit=false)
   &VC_FilteredSocketCount()
   &SpawnRate(SpawnRate=50, Use Spawn Probability=false) {enabled=false}
  ]
  particleSpawnScript [
   &VC_PaintedVariantsOverride(
    Cobalt=<rgba 0.2 0.4667 1 1>
    Crimson=<rgba 1.2 0.12 0.21 0.6>
    DefaultColor=&SelectLinearColorFromArray(
     Array Sampling Mode=::NewEnumerator0
     Direct Array Index=?
     Color Selection Array=DataInterface() {propertyValues {ColorData [<rgba 0.1 0.025 0 1>, <rgba 0.15 0.0375 0 1>]}}
     Random Seed=0
    )
    Grey=<rgba 0.15 0.15 0.15 0.5>
    Purple=<rgba 1 0.8 2 1>
    R1 BurntSienna:LinearColor=<rgba 0 0 0 1>
    R1 Cobalt:LinearColor=<rgba 0 0 0 1>
    R1 Crimson:LinearColor=<rgba 0 0 0 1>
    R1 DefaultColor=<rgba 0 0 0 1>
    R1 Grey:LinearColor=<rgba 0 0 0 1>
    R1 Purple:LinearColor=<rgba 0 0 0 1>
    R1 Saffron:LinearColor=<rgba 0 0 0 1>
    R1 SkyBlue:LinearColor=<rgba 0 0 0 1>
    R1 TitaniumWhite:LinearColor=<rgba 1 0.1 0.85 1>
    Saffron=<rgba 1 0.56 0.12 1>
    SkyBlue=<rgba 0.15 0.5042 1 1>
    TitaniumWhite=<rgba 1 1 1 1>
    Use R1 Painted Overrides=true
   )
   /SetVariables(Particles.RENDERTAG:NiagaraInt32=&SpawnGroup()) {isSetParameters=true}
   &InitializeParticle(
    Color=&MakeCustomLinearColorFromBool(Bool=$User.bIsPainted, True Color=$Particles.Color, False Color=<rgba 2 0.5833 0.3 1>)
    Position Offset=&AddVectorToPosition(Position=<0 0 0>, Vector=<0 0 0>)
    Sprite Size=<70 80>
    Sprite Size Mode=::NewEnumerator4
    UsePositionOffset=false
    Write Material Random=true
   )
   &PartitionParticles(Number Of Partitions=$System.PortCount)
   &SkeletalMeshLocation(
    Bone / Socket Index=$Particles.PartitionParticles.Partition
    Bone / Socket Sampling Mode=::NewEnumerator5
    MaskSpawnGroup=false
    Mesh Sampling Type=::NewEnumerator6
    Sampled Position Offset=&AddVectorToPosition(Position=<-50 0 -2>, Vector=$User.DTCarOffset)
    Skeletal Mesh=$System.SkeletalMesh
   )
   &ApplyInitialForces()
   &AddVelocity(Velocity=<-1500 0 0>) {enabled=false}
   /VehicleCosmetics/FX/Niagara/Modules/MergeEmitters/VC_ExtraBindingColors(ExtraColors=1, R1 Color:LinearColor=$Particles.R1 Color)
   /SetVariables(Particles.Sparks SIZE 2:Vector2f=<35 45>) {isSetParameters=true, module=SetVariables_#2}
   &AlignSpriteToMeshOrientation(
    Mesh Orientation Relative Sprite Alignment Vector=<1 0 0>
    Mesh Orientation Relative Sprite Facing Vector:Vector3f=<0 0 1>
    SetAlignment=true
    SetFacing=true
   )
   &ShapeLocation(MaskSpawnGroup=false, Override Seed=false, Sphere Radius=2, UseOwnerScale=true)
   &CameraOffset(Camera Offset Amount=20)
   &DynamicMaterialParameters(
    Index 0 Param 1=&MakeCustomFloatFromBool(Bool=$User.bIsFrontendPreview, True Float=0, False Float=1)
    Module.Index0Param0Write=true
    Module.Index0Param1Write=true
    Module.Index0Param2Write=true
    Module.Index0Param3Write=true
   ) {module=DynamicMaterialParameters001}
  ]
  particleUpdateScript [
   &ParticleState(Kill Particles When Lifetime Has Elapsed=false, Loop Particles Lifetime=true)
   &SolveForcesAndVelocity(Clamp Velocity=false, Limit Acceleration=false)
   &ScaleColor(
    Scale RGB=&Multiply_VectorByFloat(Vector=<1 1 1>, Float=&MakeCustomFloatFromBool(Bool=$User.bIsPainted, True Float=2, False Float=1))
    ScaleA=true
    ScaleRGB=true
   )
  ]
  rendererClasses [&NiagaraSpriteRendererProperties]
  renderers [
   {rendererClass=&NiagaraSpriteRendererProperties}
   {rendererClass=&NiagaraSpriteRendererProperties, rendererIndex=1}
   {rendererClass=&NiagaraSpriteRendererProperties, rendererIndex=2}
   {rendererClass=&NiagaraSpriteRendererProperties, rendererIndex=3}
  ]
 }
 NE_VC_Boost_Cone_STARLINK002 {
  emitterUpdateScript [
   &EmitterState(Inactive Response=::NewEnumerator1, Life Cycle Mode=::NewEnumerator1, Loop Duration=0.1, UseLoopDelay=false)
   &SpawnBurst_Instantaneous(Spawn Group=1, Use Spawn Probability=false, UseLoopCountLimit=false) {module=SpawnBurst_Instantaneous001}
   &SpawnBurst_Instantaneous(Use Spawn Probability=false, UseLoopCountLimit=false)
   &VC_FilteredSocketCount()
   &SpawnRate(SpawnRate=50, Use Spawn Probability=false) {enabled=false}
  ]
  particleSpawnScript [
   &VC_PaintedVariantsOverride(
    Cobalt=<rgba 0.36 0.84 1.8 1>
    Crimson=<rgba 3.5 0.35 0.6125 0.6>
    DefaultColor=&SelectLinearColorFromArray(
     Array Sampling Mode=::NewEnumerator0
     Direct Array Index=?
     Color Selection Array=DataInterface() {propertyValues {ColorData [<rgba 0.1 0.025 0 1>, <rgba 0.15 0.0375 0 1>]}}
     Random Seed=0
    )
    Grey=<rgba 0.3 0.3 0.3 0.5>
    Purple=<rgba 2.5 2 5 1>
    R1 BurntSienna:LinearColor=<rgba 0 0 0 1>
    R1 Cobalt:LinearColor=<rgba 0.0356 0.0475 0.1781 1>
    R1 Crimson:LinearColor=<rgba 0.35 0 0 1>
    R1 DefaultColor=<rgba 0 0 0 1>
    R1 Grey:LinearColor=<rgba 0 0 0 1>
    R1 Purple:LinearColor=<rgba 0.175 0.1 0.25 1>
    R1 Saffron:LinearColor=<rgba 0 0 0 1>
    R1 SkyBlue:LinearColor=<rgba 0 0 0 1>
    R1 TitaniumWhite:LinearColor=<rgba 1 0.25 0.875 1>
    Saffron=<rgba 1 0.7 0.1 1>
    SkyBlue=<rgba 0.3 1.0083 2 1>
    TitaniumWhite=<rgba 2 2 2 1>
    Use R1 Painted Overrides=true
   )
   /SetVariables(Particles.RENDERTAG:NiagaraInt32=&SpawnGroup()) {isSetParameters=true}
   &InitializeParticle(
    Color=&MakeCustomLinearColorFromBool(Bool=$User.bIsPainted, True Color=$Particles.Color, False Color=<rgba 2 0.725 0.3 1>)
    Lifetime=$Emitter.CurrentLoopDuration
    Position Offset=&AddVectorToPosition(Position=<0 0 0>, Vector=<0 0 0>)
    Sprite Size=<55 70>
    Sprite Size Mode=::NewEnumerator4
    UsePositionOffset=false
    Write Material Random=true
   )
   &PartitionParticles(Number Of Partitions=$Emitter.PortCount)
   &SkeletalMeshLocation(
    Bone / Socket Index=$Particles.PartitionParticles.Partition
    Bone / Socket Sampling Mode=::NewEnumerator5
    MaskSpawnGroup=false
    Mesh Sampling Type=::NewEnumerator6
    Sampled Position Offset=&AddVectorToPosition(Position=<-50 0 -2>, Vector=$User.DTCarOffset)
    Skeletal Mesh=$System.SkeletalMesh
   )
   &ApplyInitialForces()
   &AddVelocity(Velocity=<-1500 0 0>) {enabled=false}
   &VC_ExtraBindingColors(ExtraColors=1, R1 Color:LinearColor=$Particles.R1 Color)
   /SetVariables(Particles.Sparks SIZE 2:Vector2f=<25 35>) {isSetParameters=true, module=SetVariables_#2}
   &AlignSpriteToMeshOrientation(
    Mesh Orientation Relative Sprite Alignment Vector=<1 0 0>
    Mesh Orientation Relative Sprite Facing Vector:Vector3f=<0 0 1>
    SetAlignment=true
    SetFacing=true
   )
   &ShapeLocation(MaskSpawnGroup=false, Override Seed=false, Sphere Radius=2, UseOwnerScale=true) {enabled=false}
   &CameraOffset(Camera Offset Amount=25)
   &DynamicMaterialParameters(
    Index 0 Param 1=&MakeCustomFloatFromBool(Bool=$User.bIsFrontendPreview, True Float=0, False Float=1)
    Module.Index0Param0Write=true
    Module.Index0Param1Write=true
    Module.Index0Param2Write=true
    Module.Index0Param3Write=true
   )
  ]
  particleUpdateScript [
   &ParticleState()
   &ScaleSpriteSize(Uniform Curve Sprite Scale=DataInterface() {propertyValues {Curve {keys [
        {arriveTangent=6.0046, interpMode=RCIM_Cubic, leaveTangent=6.0046, leaveTangentWeight=0.1673, tangentMode=RCTM_User, time=0, value=0}
        {arriveTangentWeight=0.0471, interpMode=RCIM_Cubic, leaveTangentWeight=0.2875, tangentMode=RCTM_User, time=0.5, value=0.9847}
        {arriveTangent=-0.0405, interpMode=RCIM_Cubic, leaveTangent=-0.0405, tangentMode=RCTM_User, time=1, value=0}
       ]}}}) {enabled=false}
   &SolveForcesAndVelocity(Clamp Velocity=false, Limit Acceleration=false)
   &ScaleColor(
    Scale RGB=&Multiply_VectorByFloat(Vector=<1 1 1>, Float=&MakeCustomFloatFromBool(Bool=$User.bIsPainted, True Float=2, False Float=1))
    ScaleA=true
    ScaleRGB=true
   )
   &JitterPosition(Jitter Amount=1)
  ]
  rendererClasses [&NiagaraSpriteRendererProperties]
  renderers [
   {rendererClass=&NiagaraSpriteRendererProperties}
   {rendererClass=&NiagaraSpriteRendererProperties, rendererIndex=1}
   {rendererClass=&NiagaraSpriteRendererProperties, rendererIndex=2}
   {rendererClass=&NiagaraSpriteRendererProperties, rendererIndex=3}
  ]
 }
 NE_VC_Boost_Cone_Sparks001 {
  emitterUpdateScript [
   &EmitterState(Inactive Response=::NewEnumerator1, Life Cycle Mode=::NewEnumerator1, Loop Duration=1, UseLoopDelay=false)
   &VC_FilteredSocketCount()
   &SpawnRate(SpawnGroup=1, SpawnRate=50, Use Spawn Probability=false) {enabled=false}
   &SpawnRate(SpawnRate=40, Use Spawn Probability=false) {module=SpawnRate001}
   &SpawnRate(SpawnGroup=1, SpawnRate=40, Use Spawn Probability=false) {module=SpawnRate002}
  ]
  particleSpawnScript [
   &VC_PaintedVariantsOverride(
    Cobalt=<rgba 0.085 0.2125 0.85 0.8>
    Crimson=<rgba 0.8 0.016 0.0552 0.8>
    DefaultColor=&SelectLinearColorFromArray(
     Array Sampling Mode=::NewEnumerator0
     Direct Array Index=?
     Color Selection Array=DataInterface() {propertyValues {ColorData [<rgba 0.1 0.025 0 1>, <rgba 0.15 0.0375 0 1>]}}
     Random Seed=0
    )
    Grey=<rgba 0.08 0.08 0.08 0.85>
    Purple=<rgba 0.4513 0.285 0.95 0.85>
    R1 Cobalt:LinearColor=<rgba 0.16 0.2667 0.8 0.2>
    R1 Crimson:LinearColor=<rgba 0.8 0.12 0.154 0.1>
    R1 Grey:LinearColor=<rgba 0.085 0.085 0.085 0.15>
    R1 Purple:LinearColor=<rgba 0.2773 0.16 0.8 0.25>
    R1 Saffron:LinearColor=<rgba 0.6 0.38 0.072 0.15>
    R1 SkyBlue:LinearColor=<rgba 0.11 0.4028 0.55 0.2>
    R1 TitaniumWhite:LinearColor=<rgba 1 1 1 0.2>
    Saffron=<rgba 0.6 0.33 0.06 0.7>
    SkyBlue=<rgba 0.14 0.42 0.7 0.5>
    TitaniumWhite=<rgba 2.5 2.5 2.5 0.25>
    Use R1 Painted Overrides=true
   )
   /SetVariables(Particles.RENDER TAG:NiagaraInt32=&SpawnGroup()) {isSetParameters=true}
   &InitializeParticle(
    Color=&MakeCustomLinearColorFromBool(
     Bool=$User.bIsPainted
     True Color=&Multiply_LinearColorByFloat(Linear Color=$Particles.Color, Float=1)
     False Color=&SelectLinearColorFromArray(
      Array Sampling Mode=::NewEnumerator0
      Direct Array Index=?
      Color Selection Array=DataInterface() {propertyValues {ColorData [<rgba 0.1 0.025 0 1>, <rgba 0.05 0.0125 0 1>]}}
      Random Seed=0
     )
    )
    Lifetime Max=&MaskFloatBySpawnGroup(Masked Float=0.2, Passthrough Float=0.2, Spawn Group Mask=1)
    Lifetime Min:NiagaraFloat=&MaskFloatBySpawnGroup(Masked Float=0.2, Passthrough Float=0.2, Spawn Group Mask=1)
    Lifetime Mode=::NewEnumerator1
    Position Offset=&AddVectorToPosition(Position=<0 0 0>, Vector=<0 0 0>)
    Sprite Rotation Mode=::NewEnumerator2
    Sprite Size Mode=::NewEnumerator3
    Uniform Sprite Size=&MakeCustomFloatFromBool(
     Bool=$User.bIsFrontendPreview
     True Float=&SelectFloatFromArray(
      Array Sampling Mode=::NewEnumerator0
      Direct Array Index=?
      Float Selection Array=DataInterface() {propertyValues {FloatData [70,50]}}
      Random Seed=0
     )
     False Float=90
    )
    UsePositionOffset=false
    Write Material Random=true
   )
   &PartitionParticles(Number Of Partitions=$System.PortCount)
   &SkeletalMeshLocation(
    Bone / Socket Index=$Particles.PartitionParticles.Partition
    Bone / Socket Sampling Mode=::NewEnumerator5
    MaskSpawnGroup=false
    Mesh Sampling Type=::NewEnumerator6
    Sampled Position Offset=&AddVectorToPosition(Position=<-50 0 0>, Vector=$User.DTCarOffset)
    Skeletal Mesh=$System.SkeletalMesh
   )
   &ApplyInitialForces()
   &AlignSpriteToMeshOrientation(
    Mesh Orientation Relative Sprite Alignment Vector=<1 0 0>
    Mesh Orientation Relative Sprite Facing Vector:Vector3f=<0 0 1>
    SetAlignment=true
    SetFacing=true
   )
   &AddVelocity(Velocity=&MaskVectorBySpawnGroup(Masked Vector=<0 0 0>, Passthrough Vector=<50 0 0>), Velocity Speed Scale=-1) {enabled=false}
   &VC_ExtraBindingColors(
    ExtraColors=1
    R1 Color:LinearColor=&MakeCustomLinearColorFromBool(
     Bool=$User.bIsPainted
     True Color=&Multiply_LinearColorByFloat(Linear Color=$Particles.R1 Color, Float=3.5)
     False Color=&SelectLinearColorFromArray(
      Array Sampling Mode=::NewEnumerator0
      Direct Array Index=?
      Color Selection Array=DataInterface() {propertyValues {ColorData [<rgba 1.2 0.56 0.24 0.5>, <rgba 4 1.3 0.4 0.2>]}}
      Random Seed=0
     )
    )
   )
   /SetVariables(Particles.Sparks SIZE 2:Vector2f=&Multiply_Vector2D_ByFloat(
     Float=&SelectFloatFromArray(
      Array Sampling Mode=::NewEnumerator0
      Direct Array Index=?
      Float Selection Array=DataInterface() {propertyValues {FloatData [30,40]}}
      Random Seed=0
     )
     Vector2D=&MakeCustomVector2DFromBool(Bool=$User.bIsFrontendPreview, True Vector2D=<1 1>, False Vector2D=<1.5 1.5>)
    )) {isSetParameters=true, module=SetVariables_#2}
   &ShapeLocation(
    MaskSpawnGroup=false
    Override Seed=false
    Sphere Radius=&MaskFloatBySpawnGroup(Masked Float=2, Passthrough Float=0, Spawn Group Mask=1)
    UseOwnerScale=true
   ) {enabled=false}
   &CameraOffset(Camera Offset Amount=&SelectFloatFromArray(Direct Array Index=&SpawnGroup(), Float Selection Array=DataInterface() {propertyValues {FloatData [25,25]}}))
   &DynamicMaterialParameters(
    Index 0 Param 1=&MakeCustomFloatFromBool(Bool=$User.bIsFrontendPreview, True Float=0, False Float=1)
    Module.Index0Param0Write=true
    Module.Index0Param1Write=true
    Module.Index0Param2Write=true
    Module.Index0Param3Write=true
   )
  ]
  particleUpdateScript [
   &ParticleState()
   &SpriteRotationRate(Rotation Rate=&SelectFloatFromArray(
     Array Sampling Mode=::NewEnumerator0
     Direct Array Index=?
     Evaluation Type=::NewEnumerator1
     Float Selection Array=DataInterface() {propertyValues {FloatData [5,-5]}}
     Random Seed=0
    ))
   &SolveForcesAndVelocity(Clamp Velocity=false, Limit Acceleration=false)
   &ScaleColor(
    Scale Alpha=&FloatFromCurve(FloatCurve=DataInterface() {propertyValues {Curve {keys [
         {arriveTangent=-0.0568, interpMode=RCIM_Cubic, leaveTangent=-0.0568, tangentMode=RCTM_User, time=0, value=0}
         {arriveTangentWeight=0.0471, interpMode=RCIM_Cubic, leaveTangentWeight=0.2875, tangentMode=RCTM_User, time=0.5, value=0.9847}
         {arriveTangent=-0.0405, interpMode=RCIM_Cubic, leaveTangent=-0.0405, tangentMode=RCTM_User, time=1, value=0}
        ]}}})
    ScaleA=true
    ScaleRGB=true
   )
  ]
  rendererClasses [&NiagaraSpriteRendererProperties]
  renderers [
   {rendererClass=&NiagaraSpriteRendererProperties}
   {rendererClass=&NiagaraSpriteRendererProperties, rendererIndex=1}
   {rendererClass=&NiagaraSpriteRendererProperties, rendererIndex=2}
   {rendererClass=&NiagaraSpriteRendererProperties, rendererIndex=3}
  ]
 }
 NE_VC_Boost_MAIN_RIBBON002 {
  emitterSpawnScript [/SetVariables(Emitter.INGAME:NiagaraBool=&Boolean_NotOperation(Bool=$User.bIsFrontendPreview)) {isSetParameters=true}]
  emitterUpdateScript [
   &EmitterState(Inactive Response=::NewEnumerator1, Life Cycle Mode=::NewEnumerator1, Loop Duration=1, UseLoopDelay=false)
   &VC_FilteredSocketCount()
   &VC_RateQualityHandler()
   &VC_SpawnRibbon_v1(
    Number of Ribbons=$Emitter.PortCount
    Spawn Time=-1
    SpawnRate=&MakeCustomFloatFromBool(Bool=$User.bIsFrontendPreview, True Float=100, False Float=30)
   )
  ]
  particleSpawnScript [
   &VC_PaintedVariantsOverride(
    BurntSienna=<rgba 0 0 0 1>
    Cobalt=<rgba 0 0 0 1>
    Crimson=<rgba 0 0 0 1>
    DefaultColor=<rgba 0 0 0 1>
    Grey=<rgba 0 0 0 1>
    Purple=<rgba 0 0 0 1>
    R1 Black:LinearColor=<rgba 0.03 0.03 0.03 1>
    R1 BurntSienna:LinearColor=<rgba 0.0642 0.0198 0.0079 1>
    R1 Cobalt:LinearColor=<rgba 0.24 0.408 1.2 1>
    R1 Crimson:LinearColor=<rgba 1.5 0 0.0625 1>
    R1 Grey:LinearColor=<rgba 0.2 0.2 0.2 1>
    R1 Purple:LinearColor=<rgba 1.65 1.2 3 1>
    R1 Saffron:LinearColor=<rgba 2 1.4333 0.3 1>
    R1 SkyBlue:LinearColor=<rgba 0.54 1.17 1.8 1>
    R1 TitaniumWhite:LinearColor=<rgba 3.5867 3.2 4 1>
    R2 Cobalt:LinearColor=<rgba 0.306 0.5499 1.7 1>
    R2 Crimson:LinearColor=<rgba 0.8 0 0.0133 1>
    R2 Grey:LinearColor=<rgba 0.08 0.08 0.08 1>
    R2 Purple:LinearColor=<rgba 0.4905 0.315 0.9 1>
    R2 Saffron:LinearColor=<rgba 0.9 0.522 0.09 1>
    R2 SkyBlue:LinearColor=<rgba 0.3 0.675 1.2 1>
    R2 TitaniumWhite:LinearColor=<rgba 2.535 2.1 3 1>
    R3 Cobalt:LinearColor=<rgba 0.108 0.29 1.2 1>
    R3 Crimson:LinearColor=<rgba 1.2 0.036 0.0748 1>
    R3 Grey:LinearColor=<rgba 0.05 0.05 0.05 1>
    R3 Purple:LinearColor=<rgba 0.2475 0.18 0.45 1>
    R3 Saffron:LinearColor=<rgba 0.5 0.2525 0.05 1>
    R3 SkyBlue:LinearColor=<rgba 0.11 0.4512 1 1>
    R3 TitaniumWhite:LinearColor=<rgba 0.6875 0.5625 0.75 1>
    R4 Cobalt:LinearColor=<rgba 0 0 50 0.2>
    R4 Crimson:LinearColor=<rgba 1.5 0 0.075 0.1>
    R4 Grey:LinearColor=<rgba 0.02 0.02 0.02 0.1>
    R4 Purple:LinearColor=<rgba 4.1 2 20 0.1>
    R4 Saffron:LinearColor=<rgba 1 0.7667 0 0.2>
    R4 SkyBlue:LinearColor=<rgba 0 0.3125 0.75 0.1>
    R4 TitaniumWhite:LinearColor=<rgba 0.22 0.022 0.0649 0.2>
    R5 Crimson:LinearColor=<rgba 0.8 0 0.0333 0.5>
    R5 Grey:LinearColor=<rgba 0.02 0.02 0.02 0.1>
    R5 Purple:LinearColor=<rgba 0.0643 0.02 0.4 0.25>
    R5 Saffron:LinearColor=<rgba 0.5 0.275 0.05 0.2>
    R5 SkyBlue:LinearColor=<rgba 0 0.357 0.5365 0.2>
    Saffron=<rgba 0 0 0 1>
    SkyBlue=<rgba 0 0 0 1>
    TitaniumWhite=<rgba 0 0 0 0>
    Use R1 Painted Overrides=true
    Use R2 Painted Overrides=true
    Use R3 Painted Overrides=true
    Use R4 Painted Overrides=true
    Use R5 Painted Overrides=true
   )
   &VC_InitializeRibbon_v1(
    Color=$Particles.Color
    Custom Offset=true
    Enable Skeletal Position=true
    Lifetime=&MakeCustomFloatFromBool(Bool=$User.bIsFrontendPreview, True Float=0.25, False Float=1)
    Local Vector Offset=&AddVectorToPosition(
     Position=/Niagara/DynamicInputs/TypeConversions/MakeCustomPositionFromBool(Bool:NiagaraBool=$User.bIsFrontendPreview, True Position:NiagaraPosition=<20 0 0.5>, False Position:NiagaraPosition=<60 0 2>)
     Vector=$User.DTCarOffset
    )
    Offset Amount=1
    Ribbon Width=80
   )
   &AddVelocity(
    Velocity=&MakeCustomVectorFromBool(Bool=$User.bIsFrontendPreview, True Vector=<5500 0 0>, False Vector=<2500 0 0>)
    Velocity Speed Scale=-1
   ) {module=AddVelocity001}
   &VC_ExtraBindingColors(
    ExtraColors=4
    R1 Color:LinearColor=&MakeCustomLinearColorFromBool(
     Bool=$User.bIsPainted
     True Color=&Multiply_LinearColorByFloat(Linear Color=$Particles.R1 Color, Float=3)
     False Color=<rgba 1.2 0.381 0.12 1>
    )
    R2 Color:LinearColor=&MakeCustomLinearColorFromBool(
     Bool=$User.bIsPainted
     True Color=&Multiply_LinearColorByFloat(Linear Color=$Particles.R2 Color, Float=0.5)
     False Color=<rgba 0.07 0.0163 0 1>
    )
    R3 Color:LinearColor=&MakeCustomLinearColorFromBool(
     Bool=$User.bIsPainted
     True Color=&Multiply_LinearColorByFloat(Linear Color=$Particles.R3 Color, Float=3)
     False Color=<rgba 0.55 0.1705 0.055 1>
    )
    R4 Color:LinearColor=&MakeCustomLinearColorFromBool(
     Bool=$User.bIsPainted
     True Color=&Multiply_LinearColorByFloat(Linear Color=$Particles.R4 Color, Float=15)
     False Color=<rgba 2 0.6507 0.24 0.5>
    )
   )
   &CameraOffset(Camera Offset Amount=&SelectFloatFromArray(Direct Array Index=$Emitter.PortCount, Float Selection Array=DataInterface() {propertyValues {FloatData [-10000,0]}})) {enabled=false}
  ]
  particleUpdateScript [
   &ParticleState()
   &InheritSourceMovement(
    Applied Position Scale=&Multiply_VectorByFloat(Vector=<1 1 1>, Float=1)
    ApplyForce=false
    ApplyPosition=true
    ApplyPositionFalloff=false
   )
   &SolveForcesAndVelocity(Clamp Velocity=false, Limit Acceleration=false)
   &VC_UpdateRibbon_v1(Enable Skeletal Position=true)
   /SetVariables(
    Particles.R1 Color:LinearColor=&VC_DynamicScaleColor(
     Color to Scale:LinearColor=$Particles.R1 Color
     RGB Scale:Vector3f=&VectorFromFloat(Value=&FloatFromCurve(FloatCurve=DataInterface() {propertyValues {Curve {keys [<at 0 1 RCIM_Cubic>, {arriveTangentWeight=0.7535, interpMode=RCIM_Cubic, tangentMode=RCTM_User, time=1, value=0}]}}}))
     Alpha Scale:NiagaraFloat=0
    )
    Particles.R2 Color:LinearColor=&VC_DynamicScaleColor(
     Color to Scale:LinearColor=$Particles.R2 Color
     RGB Scale:Vector3f=&MakeCustomVectorFromBool(
      Bool=$User.bIsPainted
      True Vector=&VectorFromFloat(Value=&FloatFromCurve(FloatCurve=DataInterface() {propertyValues {Curve {keys [<at 0 0.5 RCIM_Cubic>, {arriveTangent=-0.7692, interpMode=RCIM_Cubic, leaveTangent=-0.7692, time=0.5, value=2.7}, <at 0.65 0 RCIM_Cubic>]}}}))
      False Vector=<1 1 1>
     )
     Alpha Scale:NiagaraFloat=&MakeCustomFloatFromBool(Bool=$User.bIsFrontendPreview, True Float=1, False Float=2)
    )
    Particles.R3 Color:LinearColor=&VC_DynamicScaleColor(
     Color to Scale:LinearColor=$Particles.R3 Color
     RGB Scale:Vector3f=&VectorFromFloat(Value=&MakeCustomFloatFromBool(
       Bool=$User.bIsPainted
       True Float=&FloatFromCurve(FloatCurve=DataInterface() {propertyValues {Curve {keys [
            <at 0 1 RCIM_Cubic>
            {arriveTangent=2, interpMode=RCIM_Cubic, leaveTangent=2, time=0.3, value=0.65}
            {arriveTangent=13.4, interpMode=RCIM_Cubic, leaveTangent=13.4, time=0.5, value=2}
            {arriveTangent=-20, interpMode=RCIM_Cubic, leaveTangent=-20, time=0.55, value=4}
            <at 0.6 0 RCIM_Cubic>
           ]}}})
       False Float=&FloatFromCurve(FloatCurve=DataInterface() {propertyValues {Curve {keys [<at 0 1 RCIM_Cubic>,<at 1 0 RCIM_Cubic>]}}})
      ))
     Alpha Scale:NiagaraFloat=&MakeCustomFloatFromBool(
      Bool=$User.bIsFrontendPreview
      True Float=&FloatFromCurve(FloatCurve=DataInterface() {propertyValues {Curve {keys [<at 0 1>,<at 0.9 0>]}}})
      False Float=&FloatFromCurve(FloatCurve=DataInterface() {propertyValues {Curve {keys [<at 0.02 0.5>,<at 0.05 1>,<at 1 0>]}}})
     )
    )
    Particles.R5 Color:LinearColor=&MakeCustomLinearColorFromBool(
     Bool=$User.bIsPainted
     True Color=&VC_DynamicScaleColor(
      Color to Scale:LinearColor=$Particles.R5 Color
      RGB Scale:Vector3f=&Multiply_VectorByFloat(
       Vector=<1 1 1>
       Float=&FloatFromCurve(
        FloatCurve=DataInterface() {propertyValues {Curve {keys [{arriveTangent=-1.1182, interpMode=RCIM_Cubic, leaveTangent=-1.1182, tangentMode=RCTM_User, time=0.1, value=1}, <at 1 0 RCIM_Cubic>]}}}
        Scale Curve=5
       )
      )
      Alpha Scale:NiagaraFloat=&FloatFromCurve(FloatCurve=DataInterface() {propertyValues {Curve {keys [<at 0 1>,<at 0.9 0>]}}}, Scale Curve=3)
     )
     False Color=&ColorFromCurve(ColorCurve=DataInterface() {propertyValues {
        AlphaCurve {keys [<at 0.3454 2 RCIM_Cubic>,<at 0.896 0 RCIM_Cubic>]}
        BlueCurve {keys [<at 0 0.25 RCIM_Cubic>]}
        GreenCurve {keys [<at 0 0.6458 RCIM_Cubic>]}
        RedCurve {keys [<at 0 5 RCIM_Cubic>]}
       }})
    )
    Particles.RIBBONWIDTH001:NiagaraFloat=&FloatFromCurve(
     FloatCurve=DataInterface() {propertyValues {Curve {keys [
         <at 0 0 RCIM_Cubic>
         {arriveTangent=-3.5624, interpMode=RCIM_Cubic, leaveTangent=-3.5624, tangentMode=RCTM_User, time=0.1, value=1.3}
         {interpMode=RCIM_Cubic, leaveTangent=-1.5637, leaveTangentWeight=0.2375, tangentMode=RCTM_Break, time=0.3, value=0.8}
         {arriveTangent=-0.4286, interpMode=RCIM_Cubic, leaveTangent=-0.4286, time=0.7, value=0.1}
         {interpMode=RCIM_Cubic, tangentMode=RCTM_User, time=1, value=0.5}
        ]}}}
     Scale Curve=&SelectFloatFromArray(Direct Array Index=$Particles.SpawnGroup, Float Selection Array=DataInterface() {propertyValues {FloatData [-200,200]}})
    )
    Particles.RIBBONWIDTH002:NiagaraFloat=&FloatFromCurve(
     FloatCurve=DataInterface() {propertyValues {Curve {keys [
         <at 0 0 RCIM_Cubic>
         {arriveTangent=-9.9818, interpMode=RCIM_Cubic, leaveTangent=-9.9818, tangentMode=RCTM_User, time=0.1, value=1.2}
         {arriveTangent=-6.1884, interpMode=RCIM_Cubic, leaveTangent=-1.5637, leaveTangentWeight=0.2375, tangentMode=RCTM_Break, time=0.3, value=0.4}
         {arriveTangent=5.6635, interpMode=RCIM_Cubic, leaveTangent=8.7287, tangentMode=RCTM_Break, time=0.4, value=0.25}
         {interpMode=RCIM_Cubic, tangentMode=RCTM_User, time=0.8, value=0}
        ]}}}
     Scale Curve=&SelectFloatFromArray(Direct Array Index=$Particles.SpawnGroup, Float Selection Array=DataInterface() {propertyValues {FloatData [-150,150]}})
    )
    Particles.RIBBONWIDTH003:NiagaraFloat=&FloatFromCurve(
     FloatCurve=DataInterface() {propertyValues {Curve {keys [
         {interpMode=RCIM_Cubic, leaveTangentWeight=0.4706, tangentMode=RCTM_User, time=0, value=0.8}
         {interpMode=RCIM_Cubic, tangentMode=RCTM_User, time=0.8, value=0}
        ]}}}
     Scale Curve=&SelectFloatFromArray(Direct Array Index=$Particles.SpawnGroup, Float Selection Array=DataInterface() {propertyValues {FloatData [-200,200]}})
    )
    Particles.RIBBONWIDTH004:NiagaraFloat=&FloatFromCurve(
     FloatCurve=DataInterface() {propertyValues {Curve {keys [
         <at 0.05 0 RCIM_Cubic>
         {arriveTangent=-0.3107, interpMode=RCIM_Cubic, leaveTangent=-0.3107, tangentMode=RCTM_User, time=0.18, value=0.95}
         <at 0.85 0 RCIM_Cubic>
        ]}}}
     Scale Curve=&SelectFloatFromArray(Direct Array Index=$Particles.SpawnGroup, Float Selection Array=DataInterface() {propertyValues {FloatData [-30,30]}})
    )
    Particles.RIBBONWIDTH_2:NiagaraFloat=&FloatFromCurve(
     FloatCurve=DataInterface() {propertyValues {Curve {keys [
         {interpMode=RCIM_Cubic, leaveTangentWeight=0.3246, tangentMode=RCTM_User, time=0.1, value=1}
         {arriveTangent=0.0471, arriveTangentWeight=0.4789, interpMode=RCIM_Cubic, leaveTangent=0.0471, tangentMode=RCTM_User, time=0.85, value=0}
        ]}}}
     Scale Curve=75
    )
   ) {isSetParameters=true}
   &DynamicMaterialParameters(
    Index 0 Param 1=2
    Index 0 Param 2=3
    Index 0 Param 3=5
    Index 0 Param 4=0.04
    Module.Index0Param0Write=true
    Module.Index0Param1Write=true
    Module.Index0Param2Write=true
    Module.Index0Param3Write=true
    Module.Index1Param0Write=true
    Module.Index1Param1Write=true
    Module.Index1Param2Write=true
    Module.Index1Param3Write=true
    Write Parameter Index 1=true
   )
   &ScaleRibbonWidth(Ribbon Width Scale=&FloatFromCurve(
     FloatCurve=DataInterface() {propertyValues {Curve {keys [
         {arriveTangent=-0.7463, interpMode=RCIM_Cubic, leaveTangent=-0.7463, leaveTangentWeight=0.357, tangentMode=RCTM_User, time=0.1, value=1}
         {arriveTangent=-0.9193, arriveTangentWeight=0.4419, interpMode=RCIM_Cubic, leaveTangent=-0.9193, tangentMode=RCTM_User, time=0.85, value=0}
        ]}}}
     Scale Curve=1.2
    ))
   &ScaleColor(
    Scale Alpha=&FloatFromCurve(FloatCurve=DataInterface() {propertyValues {Curve {keys [<at 0.8 1 RCIM_Cubic>,<at 1 0 RCIM_Cubic>]}}})
    Scale RGB=&Multiply_VectorByFloat(Vector=<1 1 1>, Float=&MakeCustomFloatFromBool(Bool=$User.bIsPainted, True Float=5, False Float=1))
    ScaleA=true
    ScaleRGB=true
   )
  ]
  rendererClasses [&NiagaraRibbonRendererProperties]
  renderers [
   {rendererClass=&NiagaraRibbonRendererProperties}
   {rendererClass=&NiagaraRibbonRendererProperties, rendererIndex=1}
   {rendererClass=&NiagaraRibbonRendererProperties, rendererIndex=2}
   {rendererClass=&NiagaraRibbonRendererProperties, rendererIndex=3}
   {rendererClass=&NiagaraRibbonRendererProperties, rendererIndex=4}
   {rendererClass=&NiagaraRibbonRendererProperties, rendererIndex=5}
   {rendererClass=&NiagaraRibbonRendererProperties, rendererIndex=6}
   {rendererClass=&NiagaraRibbonRendererProperties, rendererIndex=7}
   {rendererClass=&NiagaraRibbonRendererProperties, rendererIndex=8}
   {rendererClass=&NiagaraRibbonRendererProperties, rendererIndex=9}
   {rendererClass=&NiagaraRibbonRendererProperties, rendererIndex=10}
  ]
 }
 NE_VC_Boost_Sprite_Spheres {
  simTarget=GPUComputeSim
  emitterUpdateScript [
   &EmitterState(Inactive Response=::NewEnumerator1, Life Cycle Mode=::NewEnumerator1, Loop Duration=1, UseLoopDelay=false)
   &VC_FilteredSocketCount()
   &SpawnRate(
    SpawnRate=&MakeCustomFloatFromBool(Bool=$User.bIsFrontendPreview, True Float=150, False Float=300)
    Use Spawn Probability=false
   )
  ]
  particleSpawnScript [
   &VC_PaintedVariantsOverride(Grey=<rgba 0.1 0.1 0.1 0.2>, TitaniumWhite=<rgba 1 0.2 0.8667 0.2>)
   &InitializeParticle(
    Color=&MakeCustomLinearColorFromBool(
     Bool=$User.bIsPainted
     True Color=&Multiply_LinearColorByFloat(Linear Color=$Particles.Color, Float=10)
     False Color=<rgba 3 0.85 0 1>
    )
    Lifetime Max=0.4
    Lifetime Min=0.2
    Lifetime Mode=::NewEnumerator1
    Sprite Size=&RandomRangeVector2D(Minimum=<0.5 0.5>, Maximum=<1 1>)
    Sprite Size Mode=::NewEnumerator4
    UsePositionOffset=false
    Write Material Random=true
   )
   &PartitionParticles(Number Of Partitions=$Emitter.PortCount, Partition Mode:ENiagara_PartitionMode=::NewEnumerator5)
   &SkeletalMeshLocation(
    Bone / Socket Index=$Particles.PartitionParticles.Partition
    Bone / Socket Sampling Mode=::NewEnumerator5
    MaskSpawnGroup=false
    Mesh Sampling Type=::NewEnumerator6
    Sampled Position Offset=&AddVectorToPosition(Position=<-70 0 0>, Vector=$User.DTCarOffset)
    Skeletal Mesh=$System.SkeletalMesh
   )
   &ApplyInitialForces()
   &AddVelocity(
    Velocity=&MakeCustomVectorFromBool(Bool=$User.bIsFrontendPreview, True Vector=<1500 0 0>, False Vector=<1500 0 0>)
    Velocity Speed Scale=-1
   )
   &ShapeLocation(MaskSpawnGroup=false, Override Seed=false, Sphere Radius=25, UseOwnerScale=true)
   &CameraOffset()
  ]
  particleUpdateScript [
   &ParticleState()
   &ScaleSpriteSize(
    Scale Sprite Size Mode=::NewEnumerator0
    Uniform Scale Factor=&MaskFloatBySpawnGroup(
     Masked Float=&FloatFromCurve(
      FloatCurve=DataInterface() {propertyValues {Curve {keys [{tangentMode=RCTM_User, time=0, value=1}, {tangentMode=RCTM_User, time=1, value=0.7}]}}}
      Scale Curve=&MakeCustomFloatFromBool(Bool=$User.bIsFrontendPreview, True Float=2, False Float=3)
     )
     Passthrough Float=&MakeCustomFloatFromBool(Bool=$User.bIsFrontendPreview, True Float=1, False Float=2)
    )
   )
   &Drag(
    Drag=&MaskFloatBySpawnGroup(Masked Float=1, Passthrough Float=0.5)
    Use Linear Drag=true
    Use Rotational Drag=true
   )
   &InheritSourceMovement(
    Applied Force Scale=<0.15 0.15 0.15>
    Applied Position Scale=&Multiply_VectorByFloat(
     Vector=<1 1 1>
     Float=&FloatFromCurve(FloatCurve=DataInterface() {propertyValues {Curve {keys [{interpMode=RCIM_Cubic, tangentMode=RCTM_User, time=0, value=1}, {interpMode=RCIM_Cubic, tangentMode=RCTM_User, time=1, value=0}]}}})
    )
    ApplyForce=true
    ApplyForceFalloff=false
    ApplyPosition=true
    ApplyPositionFalloff=false
    Module.LimitForceSpeed=false
   )
   &SolveForcesAndVelocity(Clamp Velocity=false, Limit Acceleration=false)
   &ScaleColor(
    Scale Alpha=&MaskFloatBySpawnGroup(
     Masked Float=1
     Passthrough Float=&FloatFromCurve(FloatCurve=DataInterface() {propertyValues {Curve {keys [<at 0 1 RCIM_Cubic>, {arriveTangent=-0.709, interpMode=RCIM_Cubic, leaveTangent=-0.709, tangentMode=RCTM_User, time=1, value=0}]}}})
    )
    Scale RGB=&VectorFromFloat(Value=&MakeCustomFloatFromBool(Bool=$User.bIsPainted, True Float=100, False Float=100))
    ScaleA=true
    ScaleRGB=true
   )
  ]
  rendererClasses [&NiagaraSpriteRendererProperties]
  renderers [{rendererClass=&NiagaraSpriteRendererProperties}]
 }
}
summary {
 systemName=NS_Boost_S16
 emitters [
  {emitterName=NE_VC_Boost_Sprite_Spheres, simTarget=GPUComputeSim, rendererClasses [&NiagaraSpriteRendererProperties]}
  {emitterName=NE_VC_Boost_Cone_STARLINK001, rendererClasses [&NiagaraSpriteRendererProperties]}
  {emitterName=NE_VC_Boost_MAIN_RIBBON002, rendererClasses [&NiagaraRibbonRendererProperties]}
  {emitterName=NE_VC_Boost_Cone_Sparks001, rendererClasses [&NiagaraSpriteRendererProperties]}
  {emitterName=NE_VC_Boost_Cone_STARLINK002, rendererClasses [&NiagaraSpriteRendererProperties]}
 ]
 userVariables [
  User.bIsFrontendPreview {type=NiagaraBool, defaultValue=true}
  User.bIsPainted {type=NiagaraBool}
  User.ColorVariantID {type=NiagaraInt32}
  User.DTCarOffset {type=Vector3f, description="", defaultValue {struct=Vector3f, value=<0 0 0>}}
  User.PaintedColor {type=LinearColor, description="", defaultValue {struct=LinearColor, value=<rgba 1 1 1 1>}}
  User.PaintedScalar {type=NiagaraFloat, defaultValue=1, description=""}
  User.PortCount {type=NiagaraInt32, defaultValue=2}
 ]
}
