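# Translucent, unlit, two-sided Niagara sprite material for the cone sparks of the S16 vehicle boost.
# Particle color RGB feeds the shared FX emissive adjustment function, while particle alpha multiplies the red channel of the cone texture parameter.
# A dynamic "IS INGAME?" parameter switches an offset: above one half it uses a clamped camera offset plus a SORT ORDER parameter, otherwise a fixed vector.
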Material /VehicleCosmetics/Boosts/Boost_S16/Materials/M_Boost_Cone_Sparks
graph {
 properties {
  bAutomaticallySetUsageInEditor=false
  bUsedWithNiagaraSprites=true
  blendMode=BLEND_Translucent
  shadingModel=MSM_Unlit
  twoSided=true
  usedShadingModels=Unlit
 }
 nodes [
  n1 = Constant3Vector() {constant=<rgba 1 10 5 0>}
  n2 = Constant() {r=30}
  If(
   A=Constant() {r=0.5}
   B=DynamicParameter() {paramNames [IS INGAME?, Param2, Param3, Param4]}.IS INGAME?
   "A > B"=@n1
   "A == B"=@n1
   "A < B"=Add(
    A=ScalarParameter() {parameterName=SORT ORDER}
    B=MaterialFunctionCall(Offset Amount=@n2, Clamp Padding=@n2) {materialFunction=/Engine/Functions/Engine_MaterialFunctions02/WorldPositionOffset/CameraOffset}.Clamped Camera Offset
   )
  ) {constB=0.1}
  n3 = ParticleColor()
  MaterialFunctionCall(Emissive=@n3.RGB) {materialFunction=&MF_FX_EmissiveAdjustments}
  Multiply(
   A=@n3.A
   B=TextureSampleParameter2D() {parameterName=CONE TEXTURE, texture=/DealershipSharedLibrary/Assets/Sets/S16/T_S16_Zairo_FlipbookStart_01}.R
  )
 ]
}
