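# Unlit translucent Niagara ribbon material for the star trail of vehicle Boost S16.
# Separate color and alpha static switches select spears, inner, main or fallback trail looks.
# The main look scrolls a star texture offset by camera vector and normal distortion, shaped by masks and UV fades.
# The fallback hard-thresholds channel-selected trail textures, then subtracts noise and tip-mask dissolve, with dynamic parameters controlling speed and dissolve.
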Material /VehicleCosmetics/Boosts/Boost_S16/Materials/M_Boost_StarTrail
graph {
 properties {
  bAutomaticallySetUsageInEditor=false
  bUsedWithNiagaraRibbons=true
  blendMode=BLEND_Translucent
  refractionMethod=RM_PixelNormalOffset
  shadingModel=MSM_Unlit
  usedShadingModels=Unlit
 }
 nodes [
  n1 = TextureCoordinate()
  n2 = Power(Base=OneMinus(None=ComponentMask(None=@n1) {r=true}), Exp=ScalarParameter() {defaultValue=1, parameterName=Contrast})
  n3 = ScalarParameter() {defaultValue=1, parameterName=Brightness}
  MaterialExpressionConstantClamp_4 = Saturate(None=Multiply(A=Multiply(A=@n2, B=@n3), B=ScalarParameter() {defaultValue=1, parameterName=Strength}))
  MaterialExpressionVertexColor_0 = ParticleColor()
  n4 = Multiply(A=@MaterialExpressionConstantClamp_4, B=@MaterialExpressionVertexColor_0.RGB)
  n5 = OneMinus(None=ComponentMask(None=TextureCoordinate()) {r=true})
  MaterialFunctionCall!(Base=@n4, Add=Multiply(A=Power(Base=@n5) {constExponent=5}) {constB=2}) {materialFunction=/Engine/Functions/Engine_MaterialFunctions02/Math/MultiplyAdd}
  n6 = TextureCoordinate() {uTiling=3, vTiling=3}
  n7 = Multiply(
   A=TextureSample(UVs=Panner(Coordinate=@n6) {speedX=-2}) {texture=/DealershipSharedLibrary/FX/Textures/Noise/T_VCVFX_Boost_S16_PerlinBlur}.RGB
   B=ScalarParameter() {defaultValue=0.04, parameterName=StrengthDistort}
  )
  n8 = TextureCoordinate() {uTiling=0.3}
  n9 = Constant2Vector() {g=-0.01, r=-0.01}
  n10 = TextureSample(UVs=Add(A=ComponentMask(None=@n7) {g=true, r=true}, B=Add(A=@n8, B=@n9))) {texture=/VehicleCosmetics/Boosts/Boost_S16/Textures/T_VCVFX_Boost_S16_ZairoInTrail_01_RGB}
  n11 = DynamicParameter() {parameterIndex=1, paramNames [speed, offset, EndDissolve, UvDistortAmount]}
  n12 = AppendVector(A=@n11.offset, B=ScalarParameter() {parameterName=append param})
  n13 = Add(A=TextureCoordinate(), B=Constant2Vector() {r=-0.05})
  n14 = Time()
  n15 = TextureSample(UVs=Panner(Coordinate=Add(A=@n12, B=@n13), Time=Multiply(A=@n11.speed, B=@n14)) {speedX=-2}) {texture=&T_VCVFX_Boost_S16_ZairoInTrail_01_RGB}
  n16 = Constant()
  n17 = Add(
   A=Add(
    A=StaticSwitchParameter(True=@n15.R, False=@n16) {defaultValue=true, parameterName=Red}
    B=StaticSwitchParameter(True=@n15.G, False=@n16) {parameterName=Green}
   )
   B=Multiply(A=StaticSwitchParameter(True=@n15.B, False=@n16) {parameterName=Blue}, B=Constant() {r=3})
  )
  MaterialExpressionFloor_1 = Floor(None=Add(A=Add(A=@n10.G, B=@n17), B=Constant() {r=0.5}))
  MaterialExpressionConstant_24 = Constant() {r=1}
  n18 = Custom!(x=@MaterialExpressionFloor_1, a=@MaterialExpressionConstant_24) {
   code="return x >= a;"
   outputType=CMOT_Float1
   inputs [
    {inputName=x, input {expression=&self:MaterialExpressionFloor_1}}
    {inputName=a, input {expression=&self:MaterialExpressionConstant_24}}
   ]
  }
  n19 = Constant() {r=0.1}
  n20 = ParticleColor()
  n21 = ComponentMask(None=TextureCoordinate()) {r=true}
  n22 = Constant() {r=2}
  n23 = Multiply(
   A=Multiply(A=Power(Base=@n18, Exp=@n19), B=@n20.A)
   B=Power(Base=Multiply(A=@n21, B=@n22), Exp=ScalarParameter() {defaultValue=1.8, parameterName=power fade})
  )
  n24 = TextureCoordinate()
  n25 = Power(Base=OneMinus(None=ComponentMask(None=@n24) {r=true}), Exp=ScalarParameter() {defaultValue=1.2, parameterName=power fade one minus})
  n26 = Multiply(A=Multiply(A=Multiply(A=@n23, B=@n25), B=Constant() {r=3}), B=@n18)
  n27 = TextureCoordinate() {uTiling=0.8, vTiling=0.0125}
  n28 = Multiply(A=Time(), B=@n11.EndDissolve)
  n29 = Panner(Coordinate=TextureCoordinate() {uTiling=1.4, vTiling=0.7}) {speedX=-2, speedY=0.25}
  n30 = TextureSample(UVs=Panner(Coordinate=TextureCoordinate()) {speedX=-1, speedY=0.5}) {texture=/DealershipSharedLibrary/FX/Textures/Noise/T_VCVFX_BlobNoiseGlow_A}
  n31 = TextureCoordinate()
  n32 = Multiply(A=Power(Base=ComponentMask(None=@n31) {r=true}, Exp=Constant() {r=12}), B=Constant() {r=40})
  n33 = Add(
   A=TextureSample(UVs=Panner(Coordinate=@n27, Time=@n28) {speedY=1}) {texture=/DealershipSharedLibrary/FX/Textures/Gradient_Masks/T_VCVFX_RibbonTipMask_A}.RGB
   B=Multiply(A=TextureSample(UVs=@n29) {texture=&T_VCVFX_BlobNoiseGlow_A}.RGB, B=Multiply(A=@n30.RGB, B=@n32))
  )
  n34 = OneMinus(None=ComponentMask(None=TextureCoordinate()) {r=true})
  n35 = Constant() {r=3}
  n36 = Multiply(A=Multiply(A=Subtract(A=@n26, B=@n33), B=Constant() {r=1.3}), B=Multiply(A=Power(Base=@n34, Exp=@n35), B=Constant() {r=20}))
  Multiply(A=Multiply(A=@n36, B=DepthFade()))
  n37 = ComponentMask(None=TextureCoordinate()) {r=true}
  n38 = ScalarParameter() {defaultValue=0.75, parameterName=FadeInMaskPower}
  MaterialExpressionConstantClamp_2 = Saturate(None=Multiply(A=ScalarParameter() {defaultValue=50, parameterName=FadeInMaskMult}, B=Power(Base=@n37, Exp=@n38)))
  n39 = Time()
  n40 = DynamicParameter() {paramNames [AlphaMult,Emissive,ColorAdd,TextureSpeed]}
  n41 = Multiply(A=@n40.TextureSpeed, B=Constant() {r=-2})
  n42 = Add(A=Multiply(A=Multiply(A=@n39, B=@n41), B=Constant2Vector()), B=TextureCoordinate())
  n43 = TextureSample(UVs=@n42) {texture=/VehicleCosmetics/Boosts/Boost_S16/Textures/T_VCVFX_Boost_S16_ZairoMaskDistort_01}
  n44 = Clamp(None=Subtract(A=@n43.R, B=Constant() {r=0.12}), Min=Constant(), Max=Constant() {r=1})
  n45 = Power(Base=@n43.R, Exp=Constant() {r=1})
  MaterialExpressionConstantClamp_1 = Saturate(None=Multiply(A=@MaterialExpressionConstantClamp_2, B=Add(A=@n44, B=@n45)))
  MaterialExpressionDepthBiasBlend_0 = DepthFade()
  n46 = Multiply(
   A=OneMinus(None=TextureSample() {texture=/DealershipSharedLibrary/FX/Textures/Gradient_Masks/T_VCVFX_boost_Space_Tear_SquareCloud_Mask}.RGB)
   B=StaticSwitchParameter(True=Multiply(A=@MaterialExpressionConstantClamp_1, B=@MaterialExpressionDepthBiasBlend_0), False=@MaterialExpressionConstantClamp_1) {parameterName=UseDepthBiasBlend}
  )
  n47 = ComponentMask(None=TextureCoordinate()) {r=true}
  n48 = Multiply(A=@n46, B=Multiply(A=OneMinus(None=@n47)) {constB=5})
  Reroute(None=@n48)
  n49 = DotProduct(A=TextureCoordinate(), B=Constant2Vector() {r=1})
  n50 = Constant() {r=1}
  n51 = TextureSample(UVs=Multiply(A=Power(Base=@n49, Exp=@n50), B=@n42)) {samplerType=SAMPLERTYPE_Normal, texture=/DealershipSharedLibrary/FX/Textures/Normal/T_VCVFX_boost_SpaceTear_N}
  Reroute(None=@n51.RGB)
  n52 = ParticleColor()
  n53 = ParticleColor()
  n54 = Multiply(A=@MaterialExpressionVertexColor_0.RGB, B=VectorParameter() {parameterName=Space Color Mult, defaultValue=<rgba 5 5 5 1>}.RGB)
  n55 = Time()
  n56 = Multiply(A=Constant() {r=1.1}, B=@n40.TextureSpeed)
  n57 = Time()
  n58 = Multiply(A=@n40.TextureSpeed, B=Constant() {r=1.1})
  n59 = Transform(None=CameraVectorWS()) {transformSourceType=TRANSFORMSOURCE_View, transformType=TRANSFORM_Local}
  n60 = Add(
   A=AppendVector(A=Multiply(A=@n55, B=@n56), B=Multiply(A=@n57, B=@n58))
   B=Multiply(A=Constant2Vector() {g=2, r=2}, B=ComponentMask(None=@n59) {g=true, r=true})
  )
  n61 = Multiply(
   A=Constant() {r=0.1}
   B=StaticComponentMaskParameter(None=Transform(None=@n51.RGB) {transformType=TRANSFORM_Camera}) {defaultG=true, defaultR=true}
  )
  n62 = Multiply(
   A=Constant() {r=0.333}
   B=TextureSample(UVs=Add(A=@n60, B=@n61)) {texture=/VehicleCosmetics/Boosts/Boost_S16/Textures/T_VCVFX_Boost_S16_ZairoStars_RGB}.RGB
  )
  n63 = TextureSample(UVs=Panner(Coordinate=TextureCoordinate())) {texture=&T_VCVFX_Boost_S16_ZairoMaskDistort_01}
  n64 = Subtract(A=Multiply(A=Multiply(A=@n54, B=@n62), B=@n40.Emissive), B=Power(Base=@n63.R, Exp=Constant() {r=0.5}))
  n65 = Constant() {r=0.1}
  n66 = OneMinus(None=Power(Base=Subtract(A=@n63.R, B=@n65), Exp=Constant() {r=0.3}))
  n67 = Constant() {r=2}
  n68 = Multiply(A=Add(B=Multiply(A=@n66, B=@n67)), B=@MaterialExpressionVertexColor_0.RGB)
  MaterialExpressionConstantClamp_3 = Saturate(None=Multiply(A=Add(A=@n64, B=@n68), B=Multiply(A=@MaterialExpressionVertexColor_0.RGB, B=@n40.ColorAdd)))
  n69 = Add(A=@MaterialExpressionConstantClamp_3, B=@n4)
  n70 = OneMinus(None=ComponentMask(None=TextureCoordinate()) {r=true})
  n71 = Multiply(A=Multiply(A=Power(Base=@n70) {constExponent=2.5}) {constB=2}, B=@n20.RGB)
  StaticSwitchParameter(
   True=MaterialFunctionCall(Emissive=@n52.RGB) {materialFunction=&MF_FX_EmissiveAdjustments}.Result
   False=StaticSwitchParameter(True=@n53.RGB, False=StaticSwitchParameter(True=@n69, False=@n71) {defaultValue=true, parameterName=IS MAIN RIBBON?}) {parameterName=IS INNER RIBBON?}
  ) {parameterName=IS SPEARS RIBBON?}
  n72 = Add(A=TextureCoordinate(), B=Constant2Vector() {r=0.1})
  n73 = MaterialFunctionCall(
   X=ScalarParameter() {defaultValue=1, group=Texture, parameterName=Tiling X, sortPriority=3}
   Y=ScalarParameter() {defaultValue=1, group=Texture, parameterName=Tiling Y, sortPriority=4}
  ) {materialFunction=&MakeFloat2}
  n74 = ScalarParameter() {group=Texture, parameterName=Speed X Panner, sortPriority=5}
  n75 = ScalarParameter() {group=Texture, parameterName=Speed Y Panner, sortPriority=6}
  n76 = TextureSampleParameter2D(UVs=Panner(
    Coordinate=Multiply(A=@n72, B=@n73.Result)
    Speed=MaterialFunctionCall(X=@n74, Y=@n75) {materialFunction=&MakeFloat2}.Result
   )) {group=Texture, parameterName=Texture, sortPriority=0, texture=/DealershipSharedLibrary/FX/Textures/Gradient_Masks/T_VCVFX_WindParticles_A}
  n77 = ComponentMask(None=TextureCoordinate()) {r=true}
  n78 = Multiply(A=@n76.RGB, B=Step(X=SmoothStep(Value=@n77) {constMax=0.7, constValue=0.1}) {constY=0.01})
  n79 = Add(A=TextureCoordinate() {uTiling=0.35}, B=Constant2Vector() {r=-0.05})
  n80 = MaterialFunctionCall(X=ScalarParameter() {defaultValue=-4, parameterName=INNER RIBBON SPEED}) {materialFunction=&MakeFloat2}
  n81 = Multiply(
   A=@n53.A
   B=TextureSample(UVs=Panner(Coordinate=@n79, Speed=@n80.Result) {speedX=-2}) {texture=&T_VCVFX_Boost_S16_ZairoInTrail_01_RGB}.B
  )
  n82 = ComponentMask(None=TextureCoordinate()) {r=true}
  n83 = OneMinus(None=@n82)
  n84 = Power(Base=@n82) {constExponent=0.75}
  n85 = ComponentMask(None=TextureCoordinate()) {r=true}
  n86 = Multiply(
   A=Multiply(A=Power(Base=@n83), B=Multiply(A=@n84) {constB=10})
   B=Step(X=SmoothStep(Value=@n85) {constMax=0.7, constValue=0.1}) {constY=0.005}
  )
  n87 = Multiply(A=@n40.AlphaMult, B=@n46)
  n88 = StaticSwitchParameter(True=Multiply(A=@MaterialExpressionVertexColor_0.A, B=@n87), False=@n87) {parameterName=UseVertexAlpha}
  n89 = Power(Base=ComponentMask(None=TextureCoordinate()) {r=true}, Exp=ScalarParameter() {defaultValue=0.3, parameterName=PowerStart})
  n90 = Constant() {r=1}
  n91 = Floor(None=Add(A=Multiply(A=@n89, B=@n90), B=Constant() {r=0.5}))
  MaterialExpressionConstantClamp_0 = Saturate(None=Multiply(A=Multiply(A=@n88, B=@n91), B=@n48))
  n92 = Saturate(None=@n36)
  StaticSwitchParameter(
   True=Saturate(None=Multiply(A=@n52.A, B=@n78))
   False=StaticSwitchParameter(
    True=Multiply(A=@n81, B=@n86)
    False=StaticSwitchParameter(True=@MaterialExpressionConstantClamp_0, False=@n92) {defaultValue=true, parameterName=IS MAIN RIBBON ALPHA?}
   ) {parameterName=IS INNER RIBBON ALPHA?}
  ) {parameterName=IS SPEARS RIBBON ALPHA?}
  n93 = Constant() {r=-20}
  n94 = MaterialFunctionCall(Offset Amount=@n93, Clamp Padding=@n93) {materialFunction=/Engine/Functions/Engine_MaterialFunctions02/WorldPositionOffset/CameraOffset}
  n95 = ScalarParameter()
  n96 = Constant3Vector() {constant=<rgba 0 0 1 0>}
  n97 = Constant3Vector()
  Transform(None=Multiply(A=Add(A=@n94.Clamped Camera Offset, B=@n95), B=StaticSwitchParameter(True=@n96, False=@n97) {defaultValue=true, parameterName=IS MASKED ?})) {transformSourceType=TRANSFORMSOURCE_View}
  WorldPosition() {worldPositionShaderOffset=WPT_CameraRelative}
 ]
}
